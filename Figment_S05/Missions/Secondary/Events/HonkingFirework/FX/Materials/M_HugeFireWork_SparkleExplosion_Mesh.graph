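# Translucent unlit mesh-particle material for the huge sparkle explosion in the HonkingFirework event.
# Vertex color, particle color and a clamped time sine create hue-shifted twinkling, with vertex alpha varying phase, shift and size per sparkle.
# A Sprite function turns mesh geometry into lens-flare billboards, pushed along normals by dynamic SpriteSize and Expansion parameters.
# Medium and High quality add scrolling vector-noise drift to the vertex offset; Low and Default skip it.
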Material /Figment_S05/Missions/Secondary/Events/HonkingFirework/FX/Materials/M_HugeFireWork_SparkleExplosion_Mesh
graph {
 properties {
  bEnableResponsiveAA=true
  bUsedWithMeshParticles=true
  blendMode=BLEND_Translucent
  opacityMaskClipValue=0.01
  pixelDepthOffsetMode=PDOM_Legacy
  shadingModel=MSM_Unlit
  twoSided=true
  usedShadingModels=Unlit
 }
 nodes [
  n1 = VertexColor()
  n2 = VertexColor()
  n3 = AppendVector(A=@n2.R, B=@n2.R)
  n4 = Multiply(A=@n1.A, B=Time())
  n5 = Clamp(None=Multiply(A=Sine(None=@n4)) {constB=50}) {maxDefault=50, minDefault=0.25}
  n6 = ParticleColor()
  MaterialFunctionCall(Hue Shift Percentage=Multiply(A=@n1.A) {constB=0.025}, Texture=Multiply(A=AppendVector(A=@n3, B=@n2.B), B=Multiply(A=@n5, B=@n6.RGB))) {materialFunction=&HueShift}
  n7 = VertexColor()
  Multiply(A=TextureSample() {texture=/Game/Effects/Fort_Effects/Textures/T_Hanover_LensFlare}.G, B=@n7.A)
  n8 = DynamicParameter() {defaultValue=<rgba 100 1 1 1>, paramNames [SpriteSize,Expansion,Param3,Param4]}
  n9 = AppendVector(A=@n8.SpriteSize, B=@n8.SpriteSize)
  n10 = Add(
   A=MaterialFunctionCall!(Center Location=WorldPosition().XYZ, 0-1 UVs, XY scale=Multiply(A=@n7.A, B=@n9), OverridePivot, "Normals (Optional)") {materialFunction=&Sprite}.World position offset
   B=Multiply(A=VertexNormalWS(), B=Multiply(A=@n7.A, B=@n8.Expansion))
  )
  n11 = WorldPosition()
  n12 = ActorPositionWS()
  n13 = ScalarParameter() {defaultValue=35, parameterName=Speed}
  n14 = Time()
  n15 = Divide(A=Subtract(A=Subtract(A=@n11.XYZ, B=@n12), B=Multiply(A=@n13, B=@n14)), B=ScalarParameter() {defaultValue=2500, parameterName=Scale})
  n16 = Add(
   A=@n10
   B=Multiply(A=ScalarParameter() {defaultValue=100, parameterName=Intensity}, B=VectorNoise(World Position=@n15) {noiseFunction=VNF_VectorALU})
  )
  QualitySwitch(Default=@n10, Low=@n10, High=@n16, Medium=@n16)
 ]
}
